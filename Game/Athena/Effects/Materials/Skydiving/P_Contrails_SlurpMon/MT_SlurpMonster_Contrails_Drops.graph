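# Translucent unlit particle master material, named for droplets in the Slurp Monster skydiving contrail.
# Color blends two tints by texture, fresnel or sphere mask, then adds cubemap or texture reflections, with optional particle color.
# Opacity combines a sphere or texture mask, optional noise erosion, particle alpha, edge mask, near-camera and depth fades.
# Dynamic parameters drive UV distortion, panning and erosion; mobile drops shape distortion and both fades, low quality just the fades.
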Material /Game/Athena/Effects/Materials/Skydiving/P_Contrails_SlurpMon/MT_SlurpMonster_Contrails_Drops
graph {
 properties {
  bUsedWithNanite=true
  bUsedWithNiagaraMeshParticles=true
  bUsedWithNiagaraSprites=true
  bUsedWithParticleSprites=true
  blendMode=BLEND_Translucent
  pixelDepthOffsetMode=PDOM_Legacy
  shadingModel=MSM_Unlit
  usedShadingModels=Unlit
 }
 nodes [
  MaterialFunctionCall() {materialFunction=&DebugTimeSine}
  n1 = TextureCoordinate()
  n2 = ParticleRandom()
  n3 = VectorParameter() {parameterName=Color Mask UV Optrions, channelNames {a=Speed Y, b=Speed X, g=Tile Y, r=Tile X}, defaultValue=<rgba 1 1 0 0>}
  n4 = Panner(Coordinate=Multiply(A=Add(A=@n1, B=@n2), B=AppendVector(A=@n3.Tile X, B=@n3.Tile X)), Speed=AppendVector(A=@n3.Speed X, B=@n3.Speed X))
  n5 = VectorParameter() {parameterName=Color Mask Optrions, channelNames {a=Contrast, b=Add, g=Mul, r=Power}, defaultValue=<rgba 1 1 0 0>}
  n6 = Multiply(
   A=Power(Base=TextureSampleParameter2D(UVs=@n4) {parameterName=Color Mask, texture=/Engine/EngineResources/WhiteSquareTexture}.R, Exp=@n5.Power)
   B=@n5.Mul
  )
  n7 = Saturate(None=MaterialFunctionCall(In=Add(A=@n6, B=@n5.Add), Contrast=@n5.Contrast) {materialFunction=&CheapContrast}.Result)
  n8 = ScalarParameter() {defaultValue=3, parameterName=FresnelExp}
  n9 = DynamicParameter() {parameterIndex=1, paramNames [Shape_UScale, Shape_VScale, Shape_UOffset, Shape_VOffset]}
  n10 = LinearInterpolate(
   A=TextureCoordinate() {vTiling=3}
   B=Multiply(A=TextureCoordinate(), B=AppendVector(A=@n9.Shape_UScale, B=@n9.Shape_UScale))
   Alpha=ScalarParameter() {group=Dynamic Params, parameterName=Use Dynamic Shape UV Scale, sortPriority=3}
  )
  n11 = LinearInterpolate(
   A=@n10
   B=Add(A=@n10, B=AppendVector(A=@n9.Shape_UOffset, B=@n9.Shape_UOffset))
   Alpha=ScalarParameter() {group=Dynamic Params, parameterName=Use Dynamic Shape UV Offset, sortPriority=2}
  )
  n12 = TextureCoordinate() {uTiling=2, vTiling=2}
  n13 = DynamicParameter() {parameterIndex=3, paramNames [Distortion_UScale, Distortion_VScale, None, None]}
  n14 = Multiply(A=TextureCoordinate(), B=AppendVector(A=@n13.Distortion_UScale, B=@n13.Distortion_UScale))
  n15 = ScalarParameter() {group=Dynamic Params, parameterName=Use Dynamic Distortion UV Scale, sortPriority=0}
  n16 = AppendVector(A=Constant(), B=Constant() {r=-0.03})
  n17 = DynamicParameter() {parameterIndex=2, paramNames [Shape_PannerX_Speed, Shape_PannerY_Speed, Distortion_PannerX_Speed, Distortion_PannerY_Speed]}
  n18 = AppendVector(A=@n17.Distortion_PannerX_Speed, B=@n17.Distortion_PannerX_Speed)
  n19 = ScalarParameter() {group=Dynamic Params, parameterName=Use Dynamic Distortion Panner, sortPriority=1}
  n20 = TextureSampleParameter2D(UVs=Panner(Coordinate=LinearInterpolate(A=@n12, B=@n14, Alpha=@n15), Speed=LinearInterpolate(A=@n16, B=@n18, Alpha=@n19)) {speedY=-0.03}) {parameterName="Distortion Texture(G)", texture=/Game/Global/Textures/Default/Tileable_Patterns/T_Tilling_Distortion_Map_01}
  n21 = DynamicParameter() {paramNames [Erode, Distortion, Distortion_UOffset, Distortion_VOffset]}
  n22 = Multiply(
   A=@n20.G
   B=StaticSwitchParameter(True=@n21.Distortion, False=ScalarParameter() {defaultValue=0.1, parameterName=Distortion Scale}) {parameterName=Use Distortion Dynamic}
  ) {constB=0.1}
  n23 = LinearInterpolate(
   A=@n22
   B=Add(A=@n22, B=AppendVector(A=@n21.Distortion_UOffset, B=@n21.Distortion_UOffset))
   Alpha=ScalarParameter() {group=Dynamic Params, parameterName=Use Dynamic Distortion UV Offset, sortPriority=0}
  )
  n24 = StaticSwitchParameter(True=ShadingPathSwitch(Default=Add(A=@n11, B=@n23), Mobile=@n11) {desc=Migrated from FeatureLevelSwitch}, False=TextureCoordinate()) {parameterName=Use Distortion}
  n25 = DynamicParameter() {parameterIndex=2, paramNames [Shape_PannerX_Speed, Shape_PannerY_Speed, Distortion_PannerX_Speed, Distortion_PannerY_Speed]}
  n26 = LinearInterpolate(
   A=@n24
   B=Panner(Coordinate=@n24, Speed=AppendVector(A=@n25.Shape_PannerX_Speed, B=@n25.Shape_PannerX_Speed))
   Alpha=ScalarParameter() {group=Dynamic Params, parameterName=Use Dynamic Shape Panner, sortPriority=4}
  )
  n27 = Transform(None=TextureSampleParameter2D(UVs=@n26) {parameterName=FresnelNormal, samplerType=SAMPLERTYPE_Normal, texture=/Game/Effects/Fort_Effects/Textures/Water/T_waterSplash_N}.RGB)
  n28 = TextureCoordinate()
  n29 = Constant() {r=0.5}
  n30 = ScalarParameter() {defaultValue=0.65, parameterName=Hardness}
  n31 = TextureSampleParameter2D(UVs=@n26) {parameterName=Shape, texture=/Game/Effects/Fort_Effects/Textures/Power_Source/T_Power_Source_Sphere_Mask}
  n32 = Power(
   Base=StaticSwitchParameter(
    True=SphereMask(A=@n28, B=@n29, Radius=@n29, Hardness=@n30) {hardnessPercent=0}
    False=StaticComponentMaskParameter(None=@n31.RGBA) {defaultB=true, parameterName=Color Reflection Fresnel Mask}
   ) {defaultValue=true, parameterName=UseSphereMask}
   Exp=@n8
  )
  n33 = StaticSwitchParameter(True=Fresnel(ExponentIn=@n8, Normal=@n27), False=@n32) {parameterName=Use Complex Color}
  n34 = TextureSampleParameter2D(UVs=Panner(Coordinate=TextureCoordinate() {uTiling=2, vTiling=2}) {speedY=-0.06}) {
   parameterName=CubeMapDistortionNormal
   samplerType=SAMPLERTYPE_Normal
   texture=/Game/Global/Textures/Default/Tileable_Patterns/T_Tilling_Distortion_Map_01_n
  }
  n35 = ReflectionVectorWS(CustomWorldNormal=@n27)
  n36 = Multiply(
   A=TextureSampleParameterCube!(UVs=Add(A=@n34.RGB, B=@n35), Apply View MipBias) {parameterName=Cubemap, samplerType=SAMPLERTYPE_LinearColor, texture=/Game/Packages/Fortress_Sky/TexturesHDR/T_AthenaSkylight}.RGB
   B=ComponentMask(None=MaterialFunctionCall() {materialFunction=/Engine/Functions/GetPostProcessFunctions/GetAmbientCubemapTint}.AmbientCubemapTint) {b=true, g=true, r=true}
  )
  n37 = VectorParameter() {parameterName=Cube Map Multiplier, defaultValue=<rgba 0.75 0.75 0.75 1>}
  n38 = Multiply(
   A=@n32
   B=StaticSwitchParameter(
    True=Multiply(A=@n36, B=@n37.RGB)
    False=TextureSampleParameter2D() {parameterName=reflection, texture=/Game/VisualThreatManager/StormVisuals/LargeClouds/T_EyeballReflection2}.G
   ) {parameterName=Use CubeMap Reflections}
  )
  n39 = ScalarParameter() {parameterName=ReflectionContrast}
  n40 = Add(
   A=LinearInterpolate(
    A=VectorParameter() {parameterName=ColorA, defaultValue=<rgba 0 0.4375 0.242 1>}.RGB
    B=VectorParameter() {parameterName=ColorB, defaultValue=<rgba 0.0344 0.1061 0.1979 1>}.RGB
    Alpha=StaticSwitchParameter(True=@n7, False=@n33) {parameterName=Use Texture Color}
   )
   B=Multiply(
    A=MaterialFunctionCall(In=@n38, Contrast=@n39) {materialFunction=&CheapContrast}.Result
    B=VectorParameter() {parameterName=ReflectionColor, defaultValue=<rgba 0 0.3 4 1>}.RGB
   )
  )
  n41 = ParticleColor()
  n42 = StaticSwitchParameter(True=Multiply(A=@n40, B=@n41.RGB), False=@n40) {parameterName=Multiply Particle Color}
  MaterialFunctionCall(Emissive=@n42) {materialFunction=&MF_FX_EmissiveAdjustments}
  n43 = Constant() {r=0.5}
  n44 = StaticSwitchParameter(
   True=SphereMask(A=TextureCoordinate(), B=@n43, Radius=@n43, Hardness=ScalarParameter() {defaultValue=0.8, parameterName=MaskHardness}) {hardnessPercent=0}
   False=StaticComponentMaskParameter(None=@n31.RGBA) {defaultA=true, parameterName=Alpha Shape Mask}
  ) {defaultValue=true, parameterName=UseSphereMask}
  n45 = Multiply(A=@n44, B=@n20.G)
  n46 = DynamicParameter() {paramNames [Erode, Distortion, Distortion_UOffset, Distortion_VOffset]}
  n47 = LinearInterpolate(Alpha=@n46.Erode) {constA=5, constAlpha=0}
  n48 = Clamp(None=Multiply(A=Power(Base=@n45, Exp=@n47) {constExponent=50}, B=LinearInterpolate(Alpha=@n46.Erode) {constAlpha=0, constB=200}))
  n49 = ScalarParameter() {parameterName=Erode Contrast}
  n50 = ScalarParameter() {defaultValue=1, parameterName=MaskContr}
  n51 = StaticSwitchParameter(
   True=Multiply(
    A=@n44
    B=MaterialFunctionCall(In=@n48, Contrast=@n49) {materialFunction=&CheapContrast}.Result
   )
   False=Saturate(None=MaterialFunctionCall(In=@n44, Contrast=@n50) {materialFunction=&CheapContrast}.Result)
  ) {parameterName=Use Erode Parameter or Cheap Contrast}
  n52 = Power(
   Base=MaterialFunctionCall(Distance=ScalarParameter() {defaultValue=90, parameterName=NearFade}) {materialFunction=&Near_Camera_Fade}.Result
   Exp=ScalarParameter() {defaultValue=5, parameterName=NearFadeExp}
  )
  n53 = ScalarParameter() {defaultValue=15, parameterName=DF_Dist}
  n54 = Constant() {r=1}
  n55 = ShadingPathSwitch(Default=Multiply(A=Saturate(None=@n52), B=DepthFade(FadeDistance=@n53)), Mobile=@n54) {desc=Migrated from FeatureLevelSwitch}
  n56 = ScalarParameter() {defaultValue=1.5, parameterName=TopEdgeMult}
  n57 = ScalarParameter() {defaultValue=3, parameterName=TopEdgeExp}
  Multiply(
   A=Multiply(A=Multiply(A=@n41.A, B=@n51), B=QualitySwitch(Default=@n55, Low=@n54, High=@n55, Medium=@n55))
   B=LinearInterpolate(
    B=MaterialFunctionCall!(UVs, Mult=@n56, Exp=@n57) {materialFunction=/Game/Effects/Fort_Effects/MaterialFunctions/MF_SquareEdgeMask}.BottomMask
    Alpha=ScalarParameter() {parameterName=Top Edge Mask Amount}
   ) {constA=1}
  )
  Multiply(
   A=@n42
   B=MaterialFunctionCall(
    Color_Morning=VectorParameter() {group=TOD, parameterName=TOD_Morning, defaultValue=<rgba 1 1 1 1>}.RGB
    Color_Day=VectorParameter() {group=TOD, parameterName=TOD_Afternoon, defaultValue=<rgba 1 1 1 1>}.RGB
    Color_Evening=VectorParameter() {group=TOD, parameterName=TOD_Evening, defaultValue=<rgba 1 1 1 1>}.RGB
    Color_Night=VectorParameter() {group=TOD, parameterName=TOD_Night, defaultValue=<rgba 1 1 1 1>}.RGB
   ) {materialFunction=&Color_Time_of_Day}.Result
  )
  Reroute(None=StaticComponentMaskParameter(None=@n31.RGBA) {defaultG=true, parameterName=Opacity Mask Mask})
 ]
}
